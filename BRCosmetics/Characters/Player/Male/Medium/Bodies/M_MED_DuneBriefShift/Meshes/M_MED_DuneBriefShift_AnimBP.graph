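# Animation Blueprint for the DuneBriefShift male medium cosmetic body, following the attached parent's pose with face bones layered from the head mesh.
# That pose feeds a Control Rig driven by parachute and melee flags, the latter set from weapon animation type.
# Fixed additive bone offsets and forearm-bound bracer constraints then adjust knee pads, elbow and bracers.
# Knee-bend translations and an emote flag are computed each update but unused in these graphs.
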AnimBlueprint /BRCosmetics/Characters/Player/Male/Medium/Bodies/M_MED_DuneBriefShift/Meshes/M_MED_DuneBriefShift_AnimBP
target_skeleton=&/M_MED_DuneBriefShift_Skeleton
graphs [
 AnimGraph {nodes [
   n1 = "Transform (Modify) Bone - Bone: def_knee_pad_fr_l"!(
    ComponentPose=Local To Component(LocalPose=Use cached pose 'CtrlRig'())
    bAlphaBoolEnabled:bool=True
    Translation="-5,-10,0"
    Rotation="0,0,0"
    Scale="0.01,0.01,0.01"
   ) -> (Pose) {
    alphaInputType=Bool
    rotationMode=BMM_Additive
    rotationSpace=BCS_BoneSpace
    scaleMode=BMM_Additive
    scaleSpace=BCS_BoneSpace
    translationMode=BMM_Additive
    translationSpace=BCS_BoneSpace
    boneToModify {boneName=def_knee_pad_fr_l}
    scale=<0.01 0.01 0.01>
    translation=<-5 -10>
   }
   n2 = "Transform (Modify) Bone - Bone: def_knee_pad_fr_r"!(ComponentPose=@n1, bAlphaBoolEnabled:bool=True, Translation="5,10,0", Rotation="0,0,0", Scale="0.01,0.01,0.01") -> (Pose) {
    alphaInputType=Bool
    rotationMode=BMM_Additive
    rotationSpace=BCS_BoneSpace
    scaleMode=BMM_Additive
    scaleSpace=BCS_BoneSpace
    translationMode=BMM_Additive
    translationSpace=BCS_BoneSpace
    boneToModify {boneName=def_knee_pad_fr_r}
    scale=<0.01 0.01 0.01>
    translation=<5 10>
   }
   n3 = "Transform (Modify) Bone - Bone: lowerarm_twist_02_r"!(ComponentPose=@n2, bAlphaBoolEnabled:bool=True, Translation="0,0,0", Rotation="0,0,0", Scale="0.1,0.1,0.1") -> (Pose) {
    alphaInputType=Bool
    rotationMode=BMM_Additive
    rotationSpace=BCS_BoneSpace
    scaleMode=BMM_Additive
    scaleSpace=BCS_BoneSpace
    translationMode=BMM_Additive
    translationSpace=BCS_BoneSpace
    boneToModify {boneName=lowerarm_twist_02_r}
    scale=<0.1 0.1 0.1>
   }
   n4 = "Transform (Modify) Bone - Bone: deform_elbow_out_base_r"!(ComponentPose=@n3, bAlphaBoolEnabled:bool=True, Translation="-9,3,0", Rotation="0,0,0", Scale="1,1,1") -> (Pose) {
    alphaInputType=Bool
    rotationMode=BMM_Additive
    rotationSpace=BCS_BoneSpace
    scaleMode=BMM_Additive
    scaleSpace=BCS_BoneSpace
    translationMode=BMM_Additive
    translationSpace=BCS_BoneSpace
    boneToModify {boneName=deform_elbow_out_base_r}
    translation=<-9 3>
   }
   n5 = "Constraint - def_bracer_l (P)"!(ComponentPose=@n4, bAlphaBoolEnabled:bool=True, ConstraintWeights_0=1) -> (Pose) {alphaInputType=Bool, boneToModify {boneName=def_bracer_l}, constraintSetup [{targetBone {boneName=lowerarm_l}}], constraintWeights [1]}
   n6 = "Constraint - def_bracer_r (P)"!(ComponentPose=@n5, bAlphaBoolEnabled:bool=True, ConstraintWeights_0=1) -> (Pose) {alphaInputType=Bool, boneToModify {boneName=def_bracer_r}, constraintSetup [{targetBone {boneName=lowerarm_r}}], constraintWeights [1]}
   n7 = "Constraint - def_bracer_tip_l (P)"!(ComponentPose=@n6, bAlphaBoolEnabled:bool=True, ConstraintWeights_0=1) -> (Pose) {alphaInputType=Bool, boneToModify {boneName=def_bracer_tip_l}, constraintSetup [{targetBone {boneName=lowerarm_l}}], constraintWeights [1]}
   n8 = "Constraint - def_bracer_tip_r (P)"!(ComponentPose=@n7, bAlphaBoolEnabled:bool=True, ConstraintWeights_0=1) -> (Pose) {alphaInputType=Bool, boneToModify {boneName=def_bracer_tip_r}, constraintSetup [{targetBone {boneName=lowerarm_r}}], constraintWeights [1]}
   n9 = "Transform (Modify) Bone - Bone: def_bracer_l"!(ComponentPose=@n8, bAlphaBoolEnabled:bool=True, Translation="-5,0,0", Rotation="5,0,0", Scale="1,1,1") -> (Pose) {
    alphaInputType=Bool
    rotationMode=BMM_Additive
    rotationSpace=BCS_BoneSpace
    scaleMode=BMM_Additive
    scaleSpace=BCS_BoneSpace
    translationMode=BMM_Additive
    translationSpace=BCS_BoneSpace
    boneToModify {boneName=def_bracer_l}
    rotation {pitch=5}
    translation {x=-5}
   }
   n10 = "Transform (Modify) Bone - Bone: def_bracer_r"!(ComponentPose=@n9, bAlphaBoolEnabled:bool=True, Translation="5,0,0", Rotation="5,0,0", Scale="1,1,1") -> (Pose) {
    alphaInputType=Bool
    rotationMode=BMM_Additive
    rotationSpace=BCS_BoneSpace
    scaleMode=BMM_Additive
    scaleSpace=BCS_BoneSpace
    translationMode=BMM_Additive
    translationSpace=BCS_BoneSpace
    boneToModify {boneName=def_bracer_r}
    rotation {pitch=5}
    translation {x=5}
   }
   n11 = "Transform (Modify) Bone - Bone: def_bracer_tip_l"!(ComponentPose=@n10, bAlphaBoolEnabled:bool=True, Translation="-3,0,1", Rotation="0,0,0", Scale="1,1,1") -> (Pose) {
    alphaInputType=Bool
    rotationMode=BMM_Additive
    rotationSpace=BCS_BoneSpace
    scaleMode=BMM_Additive
    scaleSpace=BCS_BoneSpace
    translationMode=BMM_Additive
    translationSpace=BCS_BoneSpace
    boneToModify {boneName=def_bracer_tip_l}
    translation {x=-3, z=1}
   }
   n12 = "Transform (Modify) Bone - Bone: def_bracer_tip_r"!(ComponentPose=@n11, bAlphaBoolEnabled:bool=True, Translation="3,0,-1", Rotation="0,0,0", Scale="1,1,1") -> (Pose) {
    alphaInputType=Bool
    rotationMode=BMM_Additive
    rotationSpace=BCS_BoneSpace
    scaleMode=BMM_Additive
    scaleSpace=BCS_BoneSpace
    translationMode=BMM_Additive
    translationSpace=BCS_BoneSpace
    boneToModify {boneName=def_bracer_tip_r}
    translation {x=3, z=-1}
   }
   Output Pose(Result=Component To Local(ComponentPose=@n12)) {layerGroup=None}
   n13 = Control Rig!(
    Source=Use cached pose 'StartPose'()
    Alpha
    IsParachuteOpenedAndCharUpRight:bool=Get bParachuteOpenedAndStandingUpright() -> (bParachuteOpenedAndStandingUpright:bool)
    isParachuteOpen:bool=Get bIsParachuteOpened()
    isMelee:bool=Get is Melee() -> (is Melee:bool)
   ) -> (Pose) {controlRigAssetReference {blueprintRigClass=&/M_MED_DuneBriefShift_CtrlRig.M_MED_DuneBriefShift_CtrlRig_C}}
   Save cached pose 'CtrlRig'(Pose=@n13) {cacheName=CtrlRig}
   n14 = Layered blend per bone(
    BasePose=Copy Pose From Mesh() {bCopyCurves=true, bUseAttachedParent=true}
    BlendPoses_0=Copy Pose From Mesh(SourceMeshComponent=Get Head_Mesh() -> (Head_Mesh:Skeletal Mesh Component Object Reference))
   ) {layerSetup [{branchFilters [
       {boneName=C_jaw}
       {boneName=C_lip_lower_mid}
       {boneName=L_lip_lower_outer}
       {boneName=R_lip_lower_outer}
       {boneName=L_cheek_inner}
       {boneName=R_cheek_inner}
       {boneName=L_lip_corner}
       {boneName=R_lip_corner}
       {boneName=faceAttach}
      ]}]}
   Save cached pose 'StartPose'(Pose=@n14) {cacheName=StartPose}
  ]}
 EventGraph {nodes [
   Event OnCharacterCustomizationLoaded()
   > n1 = Cast To PlayerPawn_Athena(Object=TryGetPawnOwner())
   > Set Head_Mesh(Head_Mesh:Skeletal Mesh Component Object Reference=GetSkeletalMeshForPartType(self=@n1, PartType=Head)) -> (Output_Get:Skeletal Mesh Component Object Reference)
   Event BlueprintInitializeAnimation()
   > Set Head_Mesh(Head_Mesh=GetOwnerSkeletalMeshComponent())
   Event BlueprintUpdateAnimation()
   n2 = GetDeltaTransformFromRefPose(self=GetOwnerSkeletalMeshComponent(), BoneName=thigh_r, BaseName=calf_r)
   n3 = MapRangeClamped(Value=BreakRotator(InRot=BreakTransform(InTransform=@n2).Rotation).Yaw, InRangeA=-20, InRangeB=-135, OutRangeB=10)
   > Set kneepad_R_translation(kneepad_R_translation:Vector=MakeVector(X=float / float(A=@n3, B=2), Y=@n3)) -> (Output_Get:Vector)
   n4 = GetDeltaTransformFromRefPose(self=GetOwnerSkeletalMeshComponent(), BoneName=thigh_l, BaseName=calf_l)
   n5 = MapRangeClamped(Value=BreakRotator(InRot=BreakTransform(InTransform=@n4).Rotation).Yaw, InRangeA=-20, InRangeB=-135, OutRangeB=-10)
   > Set kneepad_L_translation(kneepad_L_translation:Vector=MakeVector(X=float / float(A=@n5, B=2), Y=@n5)) -> (Output_Get:Vector)
   > n6 = Sequence()
   n7 = GetOwnerSkeletalMeshComponent()
   > n8 = Is Valid(InputObject=@n7)
   Set IsBoring(execute=@n8.Is Not Valid, IsBoring:bool=false) -> (Output_Get:bool)
   Set IsBoring(
    execute=@n8.Is Valid
    IsBoring=Montage_IsPlaying(self=GetAnimInstance(self=@n7), Montage=/BRCosmetics/Animation/Game/MainPlayer/Emotes/YouBoreMe/Emote_You_Bore_Me_CMM_M)
   )
   n9 = Cast To FortPlayerAnimInstance(execute=@n6.then_1, Object=GetAnimInstance(self=GetOwnerSkeletalMeshComponent()))
   > n10 = Switch on EFortWeaponCoreAnimation!(execute, Selection:EFortWeaponCoreAnimation Enum=Get WeaponCoreAnim(self=@n9)) -> (
    Melee:Exec
    Pistol:Exec
    Shotgun:Exec
    PaperBlueprint:Exec
    Rifle:Exec
    MeleeOneHand:Exec
    MachinePistol:Exec
    RocketLauncher:Exec
    GrenadeLauncher:Exec
    GoingCommando:Exec
    AssaultRifle:Exec
    TacticalShotgun:Exec
    SniperRifle:Exec
    TrapPlacement:Exec
    ShoulderLauncher:Exec
    AbilityDecoTool:Exec
    Crossbow:Exec
    C4:Exec
    RemoteControl:Exec
    DualWield:Exec
    AR_BullPup:Exec
    AR_ForwardGrip:Exec
    MedPackPaddles:Exec
    SMG_P90:Exec
    AR_DrumGun:Exec
    Consumable_Small:Exec
    Consumable_Large:Exec
    Balloon:Exec
    MountedTurret:Exec
    CreativeTool:Exec
    ExplosiveBow:Exec
    AshtonIndigo:Exec
    AshtonChicago:Exec
    MeleeDualWield:Exec
    Unarmed:Exec
   )
   Set is Melee(
    execute=[
     @n10.AssaultRifle
     @n10.GrenadeLauncher
     @n10.MachinePistol
     @n10.PaperBlueprint
     @n10.Pistol
     @n10.Rifle
     @n10.RocketLauncher
     @n10.SMG_P90
     @n10.Shotgun
     @n10.SniperRifle
     @n10.TacticalShotgun
    ]
    is Melee:bool=false
   ) -> (Output_Get:bool)
   Set is Melee(execute=@n10.Melee, is Melee=true)
  ]}
]
variables [
 Head_Mesh {type=SkeletalMeshComponent}
 is Melee {type=boolean}
 IsBoring {type=boolean}
 kneepad_L_translation {type=Vector}
 kneepad_R_translation {type=Vector}
]
